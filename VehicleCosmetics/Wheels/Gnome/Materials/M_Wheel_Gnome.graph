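# Masked material for the Gnome vehicle wheel, built on MF_WheelParent_Simple and finished with vehicle damage and de-rez transition functions.
# Vertex red splits the opacity mask between a LeafMask texture alpha and an inverted rim-mask green channel.
# Rim-mask blue adds GlowColor emissive, varied by two rotating noise samples.
# Switchable WPO pushes vertices along normals using world-position sine waves, a repeating UV gradient and vertex blue.
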Material /VehicleCosmetics/Wheels/Gnome/Materials/M_Wheel_Gnome
graph {
 properties {
  bAutomaticallySetUsageInEditor=false
  bUseMaterialAttributes=true
  bUsedWithMeshDeformer=true
  bUsedWithSkeletalMesh=true
  blendMode=BLEND_Masked
  pixelDepthOffsetMode=PDOM_Legacy
  usedShadingModels=Default Lit
 }
 nodes [
  n1 = MaterialFunctionCall() {materialFunction=&MF_WheelParent_Simple}
  n2 = GetMaterialAttributes(None=@n1.MaterialAttributes) {attributeGetTypes [#1, #2]}
  n3 = ComponentMask(None=@n1.RimRGBAMask) {g=true}
  n4 = LinearInterpolate(
   A=LinearInterpolate(A=@n2.Opacity Mask, B=OneMinus(None=@n3), Alpha=OneMinus(None=@n1.VertexColorR))
   B=TextureSampleParameter2D() {group=5_Special Effects, parameterName=LeafMask, texture=/VehicleCosmetics/Wheels/Gnome/Textures/T_Tire_Gnome_BC}.A
   Alpha=@n1.VertexColorR
  )
  n5 = VectorParameter() {parameterName=GlowColor, defaultValue=<rgba 1 0.8 0.56 1>}
  n6 = Clamp(None=Desaturation(None=@n5.RGB, Fraction=Constant() {r=-15}))
  n7 = Multiply(B=@n5.RGB) {constA=0.5, constB=0.5}
  n8 = Rotator(Coordinate=TextureCoordinate() {uTiling=0.5, vTiling=0.5}) {speed=0.5}
  n9 = Rotator(Coordinate=TextureCoordinate() {uTiling=0.6, vTiling=0.6}) {speed=0.5}
  n10 = Clamp(None=Add(
    A=TextureSample(UVs=@n8) {texture=/VehicleCosmetics/Wheels/_SharedTextures/FX/ContrastNoiseGradient_T}.B
    B=TextureSample(UVs=@n9) {texture=&ContrastNoiseGradient_T}.R
   ))
  n11 = LinearInterpolate(
   A=@n2.Emissive Color
   B=Clamp(None=LinearInterpolate(A=@n6, B=@n7, Alpha=@n10))
   Alpha=Multiply(A=ComponentMask(None=@n1.RimRGBAMask) {b=true}, B=OneMinus(None=@n1.VertexColorR))
  )
  n12 = Divide(A=WorldPosition().XYZ) {constB=4}
  n13 = ComponentMask(None=@n12) {g=true}
  n14 = Multiply(A=Time(), B=Constant() {r=0.75})
  n15 = ComponentMask(None=@n12) {r=true}
  n16 = Add(A=Sine(None=Add(A=@n13, B=@n14)) {period=3}, B=Sine(None=Add(A=@n15, B=@n14)) {period=2})
  n17 = TextureCoordinate() {vTiling=2}
  n18 = OneMinus(None=ComponentMask(None=Frac(None=@n17)) {g=true})
  n19 = Multiply(A=Multiply(A=ConstantBiasScale(None=@n16), B=Clamp(None=@n18)), B=VertexNormalWS())
  n20 = Multiply(
   A=VertexColor().B
   B=StaticSwitchParameter(True=Multiply(A=@n19) {constB=2}, False=Constant()) {defaultValue=true, group=5_Special Effects, parameterName=WPO}
  )
  MaterialFunctionCall(In=MaterialFunctionCall(Material=SetMaterialAttributes!(MaterialAttributes=@n2.MaterialAttributes, Opacity Mask=@n4, Emissive Color=@n11, World Position Offset=@n20) {attributeSetTypes [#1, #2, #3]}) {materialFunction=&MF_BaseVehicle_Damage}.Result) {materialFunction=&MF_VehicleDeRez_Transition}
 ]
}
