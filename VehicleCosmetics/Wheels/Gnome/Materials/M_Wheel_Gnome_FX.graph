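# Translucent lit material for the Gnome wheel FX on skeletal meshes, drawing an animated glow in a constant bright golden emissive.
# Two rotating samples of the glow texture's red channel are multiplied and squared, then combined with a smoothstepped static green channel to shape opacity.
# The Opacity parameter lerps that mask toward zero, so higher values are fainter and the default is nearly transparent.
# A panning blue-channel noise branch is left unconnected.
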Material /VehicleCosmetics/Wheels/Gnome/Materials/M_Wheel_Gnome_FX
graph {
 properties {
  bAutomaticallySetUsageInEditor=false
  bUseMaterialAttributes=true
  bUsedWithMeshDeformer=true
  bUsedWithSkeletalMesh=true
  blendMode=BLEND_Translucent
  pixelDepthOffsetMode=PDOM_Legacy
  usedShadingModels=Default Lit
 }
 nodes [
  n1 = TextureCoordinate() {uTiling=0.01, vTiling=0.01}
  n2 = Panner(Coordinate=@n1) {speedX=-0.005}
  n3 = Panner(Coordinate=@n1) {speedX=0.005, speedY=0.01}
  Clamp(None=Multiply(
    A=TextureSample(UVs=@n2) {samplerType=SAMPLERTYPE_LinearColor, texture=/VehicleCosmetics/Wheels/Gnome/Textures/T_Rim_Gnome_Glow}.B
    B=TextureSample(UVs=@n3) {samplerType=SAMPLERTYPE_LinearColor, texture=&T_Rim_Gnome_Glow}.B
   )) {minDefault=0.2}
  n4 = Multiply(
   A=TextureSample(UVs=Rotator()) {samplerType=SAMPLERTYPE_LinearColor, texture=&T_Rim_Gnome_Glow}.R
   B=TextureSample(UVs=Rotator() {speed=-0.15}) {samplerType=SAMPLERTYPE_LinearColor, texture=&T_Rim_Gnome_Glow}.R
  )
  n5 = Multiply(A=@n4, B=@n4)
  n6 = Add(
   A=SmoothStep(Value=TextureSample() {samplerType=SAMPLERTYPE_LinearColor, texture=&T_Rim_Gnome_Glow}.G) {constMin=0.1}
   B=@n5
  )
  n7 = Multiply(A=Clamp(None=Add(A=@n6, B=@n5)))
  MakeMaterialAttributes!(
   BaseColor
   Metallic
   Specular
   Roughness
   Anisotropy
   EmissiveColor=Multiply(A=Constant3Vector() {constant=<rgba 1 0.9 0.4 0>}) {constB=8}
   Opacity=LinearInterpolate(A=Power(Base=@n7) {constExponent=1.5}, B=Constant3Vector(), Alpha=ScalarParameter() {defaultValue=0.97, parameterName=Opacity}) {constA=1}
   OpacityMask
   Normal
   Tangent
   WorldPositionOffset
   SubsurfaceColor
   ClearCoat
   ClearCoatRoughness
   AmbientOcclusion
   Refraction
   CustomizedUVs_0
   CustomizedUVs_1
   CustomizedUVs_2
   CustomizedUVs_3
   CustomizedUVs_4
   CustomizedUVs_5
   CustomizedUVs_6
   CustomizedUVs_7
   PixelDepthOffset
   ShadingModel
   Displacement
  )
 ]
}
